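# Unlit, two-sided masked mesh-particle material for the emissive propeller effect on the red Modern Military 3 glider.
# Vertex red and green scale normal-direction offsets by dynamic TopRadius and BottomRadius, reshaping the mesh.
# Emissive multiplies a color parameter by particle green and optional sine-shaped UV gradients.
# Opacity combines particle alpha and those gradients with optional noise dissolve, where particle red scales noise UVs and inverted alpha drives erosion.
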Material /BRCosmetics/Gadgets/Assets/VinderTech_GliderChute/Glider_ModernMilitary_3_Red/Materials/FX/M_Glider_Modern_Military_3_Red_PropellerEmis
graph {
 properties {
  bUsedWithMeshParticles=true
  blendMode=BLEND_Masked
  pixelDepthOffsetMode=PDOM_Legacy
  shadingModel=MSM_Unlit
  twoSided=true
  usedShadingModels=Unlit
 }
 nodes [
  n1 = VertexColor()
  n2 = VertexNormalWS()
  n3 = DynamicParameter() {defaultValue=<rgba -20 38 1 1>, paramNames [TopRadius,BottomRadius,Param3,Param4]}
  Add(A=Multiply(A=Multiply(A=@n1.R, B=@n2), B=@n3.TopRadius) {constB=-20}, B=Multiply(A=@n3.BottomRadius, B=Multiply(A=@n1.G, B=@n2)) {constA=38})
  MaterialFunctionCall() {materialFunction=&DebugTimeSine}
  n4 = MaterialFunctionCall() {materialFunction=&LinearGradient}
  n5 = Multiply(A=@n4.UGradient) {constB=0.5}
  n6 = OneMinus(None=Saturate(None=Sine(None=@n5)))
  n7 = Saturate(None=Sine(None=Multiply(A=@n4.VGradient) {constB=0.5}))
  n8 = ParticleColor()
  n9 = Multiply(A=Multiply(A=Multiply(A=@n6, B=@n6), B=@n7), B=@n8.G)
  MaterialFunctionCall(Emissive=Multiply(
    A=StaticSwitchParameter(True=@n9, False=@n8.G) {parameterName=UseLinearGradient}
    B=VectorParameter() {parameterName=Color, defaultValue=<rgba 0.7969 0.1445 0.0889 1>}.RGB
   )) {materialFunction=&MF_FX_EmissiveAdjustments}
  n10 = Multiply(A=@n7, B=@n8.A)
  n11 = ParticleColor()
  n12 = TextureSampleParameter2D(UVs=Multiply(A=TextureCoordinate(), B=@n11.R)) {parameterName=ErodeTexture, sortPriority=3, texture=/Game/Effects/Fort_Effects/Textures/Tile/T_TileNoise_03}
  n13 = OneMinus(None=@n11.A)
  Multiply(
   A=StaticSwitchParameter(True=Multiply(A=@n6, B=@n10), False=@n8.A) {parameterName=UseLinearGradient}
   B=StaticSwitchParameter(
    True=MaterialFunctionCall!(Ramp=@n12.R, Dissolve=@n13, EdgeSharpness) {materialFunction=/Game/Effects/Fort_Effects/Materials/_Functions/MF_Dissolve}.ResultWithOutClamp
    False=Constant() {r=1}
   ) {defaultValue=true, parameterName=UseErode}
  )
 ]
}
